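# Behavior tree for NPCs inside the Battle Bus that keeps them idle, using the BB_NPC blackboard.
# A root Selector holds three Wait tasks, the first a long wait gated on the game phase step blackboard key equaling BusLocked.
# Otherwise it falls through to an unconditional long wait, leaving the final short wait as a fallback that is normally never reached.
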BehaviorTree /NPCCore/BehaviorTrees/BattleBus/BT_InsideBattleBus
blackboard=/Game/Athena/AI/NPCs/Base/BehaviorTree/BB_NPC
root {
 name=Selector
 class=BTComposite_Selector
 children [
  Wait {
   class=BTTask_Wait
   decorators [On Bus Locked? {
     class=BTDecorator_Blackboard
     properties {
      arithmeticOperation=Equal
      intValue=4
      nodeName=On Bus Locked?
      stringValue=EAthenaGamePhaseStep::BusLocked
      blackboardKey {selectedKeyName=AIEvaluator_Global_GamePhaseStep}
     }
    }]
   properties {waitTime {defaultValue=999}}
  }
  Wait {class=BTTask_Wait, properties {waitTime {defaultValue=999}}}
  Wait {class=BTTask_Wait, properties {waitTime {defaultValue=1}}}
 ]
 decorators []
 properties {
  children [
   {
    childComposite=null
    childTask=&self:BTTask_Wait_6
    decorators [&self:BTDecorator_Blackboard_5]
   }
   {childComposite=null, childTask=&self:BTTask_Wait_3, decorators []}
   {childComposite=null, childTask=&self:BTTask_Wait_1, decorators []}
  ]
  services []
 }
 services []
}
